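; Blueprint for the unlit "Off" variant of the Helios ArmChandelier C light source prop.
; The construction script runs the parent's setup, then loops over every static mesh component and sets material slot 0 to the Off material instance.
; The same loop body destroys the PointLight and TODLighting components, so those calls repeat for each mesh.
; The event graph holds only disabled placeholder events.
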(event EventBeginPlay
  (|CallParentFunction) ; This node is disabled and will not be called.\nDrag off pins to build functionality.
)

(event Collision|EventActorBeginOverlap
  (params (OtherActor :type Actor%20Object%20Reference))
  (|CallParentFunction OtherActor) ; This node is disabled and will not be called.\nDrag off pins to build functionality.
)

(event EventTick
  (params (DeltaSeconds :type Float%20%28single-precision%29))
  (|CallParentFunction DeltaSeconds) ; This node is disabled and will not be called.\nDrag off pins to build functionality.
)

(fn ConstructionScript
  (|CallParentFunction)
  (for _array_element (Actor|GetComponentsByClass "/Script/Engine.StaticMeshComponent")
    (Rendering|Material|SetMaterial _array_element 0 "/Game/Environments/Helios/Props/LightSource/LightSource_ArmChandeliers_A/Materials/MI_LightSource_ArmChandelier_A_Off.MI_LightSource_ArmChandelier_A_Off")
    (Components|DestroyComponent (Variables|Default|GetPointLight))
    (Components|DestroyComponent (Variables|Default|GetTODLightingComponent))))
